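# Physics asset for the M_MED_Shiny male body mesh, covering the jacket's dyn_ bones: panels, lapels and sleeve pieces.
# Capsules on pelvis, spine, clavicles and thighs form the body, with sleeves constrained to clavicles and lapels to spine_04.
# Two-segment box panels chain from spine_03, cross-linked front-to-side and, at the lower segment, side-to-back.
# Back panel constraints leave swing and twist free, while front and side panels use angle limits.
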PhysicsAsset /BRCosmetics/Characters/Player/Male/Medium/Bodies/M_MED_Shiny/Meshes/M_MED_Shiny_Dyn
bodies {
 clavicle_l [{shapeType=Capsule, length=3, radius=2, rotation=<rot 90 0 0>}]
 clavicle_r [{shapeType=Capsule, length=3, radius=2, rotation=<rot 90 0 0>}]
 dyn_jacket_bk_l_01 [{shapeType=Box, extentX=8.1664, extentY=3.5019, extentZ=9.3567, center=<5.6857 -0.8163 0.5885>, rotation=<rot 0 0 90>}]
 dyn_jacket_bk_l_02 [{shapeType=Box, extentX=11.6536, extentY=3.101, extentZ=16.1434, center=<5.7013 -0.8163 -0.4115>, rotation=<rot 0 0 90>}]
 dyn_jacket_bk_r_01 [{shapeType=Box, extentX=7.0132, extentY=3.398, extentZ=9.1143, center=<-5.6857 0 -1.0063>, rotation=<rot 0 0 90>}]
 dyn_jacket_bk_r_02 [{shapeType=Box, extentX=11.83, extentY=4.1895, extentZ=17.8735, center=<-5.7741 0 0.0001>, rotation=<rot 0 0 90>}]
 dyn_jacket_fr_l_01 [{shapeType=Box, extentX=18.6911, extentY=10, extentZ=2, center=<8.4452 0.0002 0>}]
 dyn_jacket_fr_l_02 [{shapeType=Box, extentX=13.5239, extentY=10, extentZ=2, center=<5.9013 0 0>, rotation=<rot 3.6756 0.8975 13.7322>}]
 dyn_jacket_fr_r_01 [{shapeType=Box, extentX=17.2068, extentY=23.803, extentZ=2, center=<-9.1378 2.413 0>}]
 dyn_jacket_fr_r_02 [{shapeType=Box, extentX=19.6944, extentY=24.0773, extentZ=2.5879, center=<-3.8726 2.413 0>}]
 dyn_jacket_l_01 [{shapeType=Box, extentX=13.5, extentY=11, extentZ=2, center=<7 0 1.0046>}]
 dyn_jacket_l_02 [{shapeType=Box, extentX=15.5172, extentY=10, extentZ=2, center=<5 0 1.295>, rotation=<rot 1.2636 0 0>}]
 dyn_jacket_r_01 [{shapeType=Box, extentX=13.5, extentY=11, extentZ=2, center=<-7 0 -0.6061>}]
 dyn_jacket_r_02 [{shapeType=Box, extentX=15.5172, extentY=10, extentZ=2, center=<-5 -0.0001 -1.8877>, rotation=<rot 4.6645 0 0>}]
 dyn_lapel_fr_l_01 [{shapeType=Capsule, length=3, radius=2, center=<0.5 0 0>, rotation=<rot 90 0 0>}]
 dyn_rapel_fr_r_01 [{shapeType=Capsule, length=3, radius=2, center=<-0.5 0 0>, rotation=<rot 90 0 0>}]
 dyn_sleeve_bk_l [{shapeType=Capsule, length=3, radius=2, center=<0.5 0 0>, rotation=<rot 90 0 0>}]
 dyn_sleeve_bk_r [{shapeType=Capsule, length=3, radius=2, center=<-0.5 0 0>, rotation=<rot 90 0 0>}]
 dyn_sleeve_top_l [{shapeType=Capsule, length=3, radius=2, center=<0.5 0 0>, rotation=<rot 90 0 0>}]
 dyn_sleeve_top_r [{shapeType=Capsule, length=3, radius=2, center=<-0.5 0 0>, rotation=<rot 90 0 0>}]
 pelvis [{shapeType=Capsule, length=9.8877, radius=12.4772, center=<-1.871 -1.3046 0>}]
 spine_01 [{shapeType=Capsule, length=4.9916, radius=11.5084, center=<3.967 -0.9806 0.0001>}]
 spine_03 [{shapeType=Capsule, length=3, radius=2, rotation=<rot 90 0 0>}]
 spine_04 [{shapeType=Capsule, length=3, radius=2, rotation=<rot 90 0 0>}]
 thigh_l [{shapeType=Capsule, length=30.5036, radius=8.8847, center=<-15 2.7822 0.763>, rotation=<rot 90 0 -4.2269>}]
 thigh_r [{shapeType=Capsule, length=18.2591, radius=9.7732, center=<10.1918 -3.2334 -1.3553>, rotation=<rot 89.2094 5.1626 3.5295>}]
}
constraints [
 {bone1Name=dyn_sleeve_top_l, bone2Name=clavicle_l, swing1LimitDegrees=10, swing2LimitDegrees=45, twistLimitDegrees=5}
 {bone1Name=dyn_sleeve_bk_l, bone2Name=clavicle_l, swing1LimitDegrees=10, swing2LimitDegrees=45, twistLimitDegrees=5}
 {bone1Name=dyn_sleeve_bk_r, bone2Name=clavicle_r, swing1LimitDegrees=10, swing2LimitDegrees=45, twistLimitDegrees=5}
 {bone1Name=dyn_sleeve_top_r, bone2Name=clavicle_r, swing1LimitDegrees=10, swing2LimitDegrees=45, twistLimitDegrees=5}
 {bone1Name=dyn_jacket_fr_l_02, bone2Name=dyn_jacket_fr_l_01, swing1LimitDegrees=45, swing2LimitDegrees=45, twistLimitDegrees=15}
 {bone1Name=dyn_jacket_l_02, bone2Name=dyn_jacket_l_01, swing1LimitDegrees=45, swing2LimitDegrees=45, twistLimitDegrees=15}
 {bone1Name=dyn_jacket_r_02, bone2Name=dyn_jacket_r_01, swing1LimitDegrees=45, swing2LimitDegrees=45, twistLimitDegrees=20}
 {bone1Name=dyn_jacket_fr_r_02, bone2Name=dyn_jacket_fr_r_01, swing1LimitDegrees=45, swing2LimitDegrees=45, twistLimitDegrees=30}
 {
  bone1Name=clavicle_l
  bone2Name=spine_04
  swing1LimitDegrees=45
  swing1Motion=Free
  swing2LimitDegrees=45
  swing2Motion=Free
  twistLimitDegrees=45
  twistMotion=Free
 }
 {
  bone1Name=clavicle_r
  bone2Name=spine_04
  swing1LimitDegrees=45
  swing1Motion=Free
  swing2LimitDegrees=45
  swing2Motion=Free
  twistLimitDegrees=45
  twistMotion=Free
 }
 {bone1Name=dyn_lapel_fr_l_01, bone2Name=spine_04, swing1LimitDegrees=45, swing2LimitDegrees=45, twistLimitDegrees=5}
 {bone1Name=dyn_rapel_fr_r_01, bone2Name=spine_04, swing1LimitDegrees=45, swing2LimitDegrees=45, twistLimitDegrees=5}
 {bone1Name=dyn_jacket_fr_l_01, bone2Name=spine_03, swing1LimitDegrees=45, swing2LimitDegrees=45, twistLimitDegrees=5}
 {bone1Name=dyn_jacket_l_01, bone2Name=spine_03, swing1LimitDegrees=45, swing2LimitDegrees=45, twistLimitDegrees=5}
 {bone1Name=dyn_jacket_r_01, bone2Name=spine_03, swing1LimitDegrees=45, swing2LimitDegrees=45, twistLimitDegrees=20}
 {bone1Name=dyn_jacket_fr_r_01, bone2Name=spine_03, swing1LimitDegrees=45, swing2LimitDegrees=45, twistLimitDegrees=30}
 {
  bone1Name=spine_04
  bone2Name=spine_03
  swing1LimitDegrees=45
  swing1Motion=Free
  swing2LimitDegrees=45
  swing2Motion=Free
  twistLimitDegrees=45
  twistMotion=Free
 }
 {
  bone1Name=spine_03
  bone2Name=spine_01
  swing1LimitDegrees=45
  swing1Motion=Free
  swing2LimitDegrees=45
  swing2Motion=Free
  twistLimitDegrees=45
  twistMotion=Free
 }
 {
  bone1Name=spine_01
  bone2Name=pelvis
  swing1LimitDegrees=45
  swing1Motion=Free
  swing2LimitDegrees=45
  swing2Motion=Free
  twistLimitDegrees=45
  twistMotion=Free
 }
 {bone1Name=thigh_l, bone2Name=pelvis, swing1LimitDegrees=45, swing2LimitDegrees=45, twistLimitDegrees=15}
 {bone1Name=thigh_r, bone2Name=pelvis, swing1LimitDegrees=45, swing2LimitDegrees=45, twistLimitDegrees=15}
 {bone1Name=dyn_jacket_r_02, bone2Name=dyn_jacket_fr_r_02, swing1LimitDegrees=45, swing2LimitDegrees=45, twistLimitDegrees=15}
 {bone1Name=dyn_jacket_r_01, bone2Name=dyn_jacket_fr_r_01, swing1LimitDegrees=45, swing2LimitDegrees=45, twistLimitDegrees=5}
 {bone1Name=dyn_jacket_fr_l_01, bone2Name=dyn_jacket_l_01, swing1LimitDegrees=45, swing2LimitDegrees=45, twistLimitDegrees=45}
 {bone1Name=dyn_jacket_fr_l_02, bone2Name=dyn_jacket_l_02, swing1LimitDegrees=45, swing2LimitDegrees=45, twistLimitDegrees=45}
 {
  bone1Name=dyn_jacket_bk_l_01
  bone2Name=spine_03
  swing1LimitDegrees=45
  swing1Motion=Free
  swing2LimitDegrees=45
  swing2Motion=Free
  twistLimitDegrees=45
  twistMotion=Free
 }
 {
  bone1Name=dyn_jacket_bk_l_02
  bone2Name=dyn_jacket_bk_l_01
  swing1LimitDegrees=45
  swing1Motion=Free
  swing2LimitDegrees=45
  swing2Motion=Free
  twistLimitDegrees=45
  twistMotion=Free
 }
 {
  bone1Name=dyn_jacket_bk_r_01
  bone2Name=spine_03
  swing1LimitDegrees=45
  swing1Motion=Free
  swing2LimitDegrees=45
  swing2Motion=Free
  twistLimitDegrees=45
  twistMotion=Free
 }
 {
  bone1Name=dyn_jacket_bk_r_02
  bone2Name=dyn_jacket_bk_r_01
  swing1LimitDegrees=45
  swing1Motion=Free
  swing2LimitDegrees=45
  swing2Motion=Free
  twistLimitDegrees=45
  twistMotion=Free
 }
 {
  bone1Name=dyn_jacket_bk_l_02
  bone2Name=dyn_jacket_l_02
  swing1LimitDegrees=45
  swing1Motion=Free
  swing2LimitDegrees=45
  swing2Motion=Free
  twistLimitDegrees=45
  twistMotion=Free
 }
 {
  bone1Name=dyn_jacket_bk_r_02
  bone2Name=dyn_jacket_r_02
  swing1LimitDegrees=45
  swing1Motion=Free
  swing2LimitDegrees=45
  swing2Motion=Free
  twistLimitDegrees=45
  twistMotion=Free
 }
]
