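# Post-process style AnimBP for the M_MED_Shiny male body that animates its jacket and collar over the pose copied from the attached parent mesh.
# Additive left-flap offsets apply while crouched and still, then AnimDynamics and RigidBody simulate the jacket, with skydiving force on some stages.
# Thigh constraints, collision limits and a crouch-walk LookAt shape the flaps.
# Event graph alphas from crouch, speed and right-arm lift drive flap and collar corrections.
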AnimBlueprint /BRCosmetics/Characters/Player/Male/Medium/Bodies/M_MED_Shiny/Meshes/M_MED_Shiny_AnimBP
target_skeleton=&/M_MED_Shiny_Skeleton
graphs [
 AnimGraph {nodes [
   Constraint!(ComponentPose, Alpha=1) -> (Pose) {alphaBoolBlend {blendOption=Linear}, boneToModify=~}
   n1 = "Transform (Modify) Bone - Bone: dyn_jacket_fr_l_01"(
    ComponentPose=Local To Component(LocalPose=Use cached pose 'SavedPose30346'())
    Alpha=1
    Translation="7.000,-1,1"
    Rotation="0,-50,-70"
    Scale="1,1,1"
   ) {
    rotationMode=BMM_Additive
    translationMode=BMM_Additive
    alphaBoolBlend {blendOption=Linear}
    boneToModify {boneName=dyn_jacket_fr_l_01}
    rotation {roll=-70, yaw=-50}
    translation=<7 -1 1>
   }
   n2 = "Transform (Modify) Bone - Bone: dyn_jacket_l_01"!(ComponentPose=@n1, bAlphaBoolEnabled:bool=True, Translation="0,0,0", Rotation="50,0,0", Scale="1,1,1") -> (Pose) {
    alphaInputType=Bool
    rotationMode=BMM_Additive
    translationMode=BMM_Additive
    alphaBoolBlend {blendInTime=0.1, blendOption=Linear, blendOutTime=0.1}
    boneToModify {boneName=dyn_jacket_l_01}
    rotation {pitch=50}
   }
   n3 = "Transform (Modify) Bone - Bone: dyn_jacket_fr_l_01"!(ComponentPose=@n2, bAlphaBoolEnabled:bool=Get bIsTargeting(), Translation="3,0,0", Rotation="50.000,0.000,0.000", Scale="1,1,1") -> (Pose) {
    alphaInputType=Bool
    rotationMode=BMM_Additive
    translationMode=BMM_Additive
    alphaBoolBlend {blendInTime=0.1, blendOption=Linear, blendOutTime=0.5}
    boneToModify {boneName=dyn_jacket_fr_l_01}
    rotation {pitch=50}
    translation {x=3}
   }
   n4 = Component To Local(ComponentPose=@n3)
   n5 = Use cached pose 'SavedPose30346'()
   n6 = AND Boolean(A="float <= float"!(A:double=Get PawnSpeed2D(), B:double=0) -> (ReturnValue:bool) {is=title K2Node_CallFunction}, B=Get bIsCrouching())
   n7 = Anim Dynamics - Bone: dyn_jacket_bk_01(
    ComponentPose=Local To Component(LocalPose=Blend Poses by bool(BlendPose_0=@n4, BlendPose_1=@n5, BlendTime_1=1.25, bActiveValue=@n6) {blendType=Linear})
    Alpha=1
    ExternalForce="0,0,0"
   ) {
    angularDampingOverride=1
    angularSpringConstant=600
    bAngularSpring=true
    bGravityOverrideInSimSpace=true
    bOverrideAngularDamping=true
    bUseSphericalLimits=true
    alphaBoolBlend {blendOption=Linear}
    boundBone {boneName=dyn_jacket_bk_01}
    componentAppliedLinearAccClamp=<800 800 800>
    componentLinearAccScale=<0.4 0.4 1>
    componentLinearVelScale=<0.8 0.8 2>
    physicsBodyDefinitions [{
      collisionType=CustomSphere
      sphereCollisionRadius=3
      boundBone {boneName=dyn_jacket_bk_01}
      boxExtents=<15 20 15>
      constraintSetup {angularLimitsMax=<-2 50 4>, angularLimitsMin=<-2 -50 -4>, angularTarget=<1 0 0>}
      localJointOffset=<12 0 0>
     }]
    relativeSpaceBone {boneName=pelvis}
    sphericalLimits [
     {limitRadius=10, drivingBone {boneName=spine_01}, sphereLocalOffset=<0 0 0>}
     {limitRadius=18, drivingBone {boneName=pelvis}, sphereLocalOffset=<0 5 0>}
    ]
   }
   n8 = Anim Dynamics - Bone: dyn_jacket_bk_02(ComponentPose=@n7, Alpha=1, ExternalForce="0,0,0") {
    angularDampingOverride=1
    angularSpringConstant=600
    bAngularSpring=true
    bGravityOverrideInSimSpace=true
    bOverrideAngularDamping=true
    bUseSphericalLimits=true
    alphaBoolBlend {blendOption=Linear}
    boundBone {boneName=dyn_jacket_bk_02}
    componentAppliedLinearAccClamp=<800 800 800>
    componentLinearAccScale=<0.4 0.4 1>
    componentLinearVelScale=<0.8 0.8 2>
    physicsBodyDefinitions [{
      collisionType=CustomSphere
      sphereCollisionRadius=3
      boundBone {boneName=dyn_jacket_bk_02}
      boxExtents=<15 20 15>
      constraintSetup {angularLimitsMax=<-2 50 4>, angularLimitsMin=<-2 -50 -4>, angularTarget=<1 0 0>}
      localJointOffset=<12 0 0>
     }]
    planarLimits [{drivingBone {boneName=pelvis}, planeTransform {location=<0 -9 0>, rotation=<rot 0 0 -90>}}]
    relativeSpaceBone {boneName=pelvis}
    sphericalLimits [
     {limitRadius=10, drivingBone {boneName=spine_01}, sphereLocalOffset=<0 0 0>}
     {limitRadius=18, drivingBone {boneName=pelvis}, sphereLocalOffset=<0 5 0>}
    ]
   }
   n9 = Anim Dynamics - Bone: dyn_jacket_bk_l_01(ComponentPose=@n8, Alpha=1, ExternalForce="0,0,0") {
    angularDampingOverride=1
    angularSpringConstant=600
    bAngularSpring=true
    bGravityOverrideInSimSpace=true
    bOverrideAngularDamping=true
    bUseSphericalLimits=true
    alphaBoolBlend {blendOption=Linear}
    boundBone {boneName=dyn_jacket_bk_l_01}
    componentAppliedLinearAccClamp=<800 800 800>
    componentLinearAccScale=<0.4 0.4 1>
    componentLinearVelScale=<0.8 0.8 2>
    physicsBodyDefinitions [{
      collisionType=CustomSphere
      sphereCollisionRadius=3
      boundBone {boneName=dyn_jacket_bk_l_01}
      boxExtents=<15 20 15>
      constraintSetup {angularLimitsMax=<-2 10 8>, angularLimitsMin=<-2 -50 -4>, angularTarget=<1 0 0>}
      localJointOffset=<12 0 0>
     }]
    planarLimits [{drivingBone {boneName=pelvis}, planeTransform {location=<0 -8.5 -8.5>, rotation=<rot 0 0 -120>}}]
    relativeSpaceBone {boneName=pelvis}
    sphericalLimits [
     {limitRadius=10, drivingBone {boneName=spine_01}, sphereLocalOffset=<0 0 -1>}
     {limitRadius=18, drivingBone {boneName=pelvis}, sphereLocalOffset=<0 5 -1>}
    ]
   }
   n10 = Anim Dynamics - Bone: dyn_jacket_bk_l_02(ComponentPose=@n9, Alpha=1, ExternalForce="0,0,0") {
    angularDampingOverride=1
    angularSpringConstant=600
    bAngularSpring=true
    bGravityOverrideInSimSpace=true
    bOverrideAngularDamping=true
    bUseSphericalLimits=true
    alphaBoolBlend {blendOption=Linear}
    boundBone {boneName=dyn_jacket_bk_l_02}
    componentAppliedLinearAccClamp=<800 800 800>
    componentLinearAccScale=<0.4 0.4 1>
    componentLinearVelScale=<0.8 0.8 2>
    physicsBodyDefinitions [{
      collisionType=CustomSphere
      sphereCollisionRadius=3
      boundBone {boneName=dyn_jacket_bk_l_02}
      boxExtents=<15 20 15>
      constraintSetup {angularLimitsMax=<-2 25 4>, angularLimitsMin=<-2 -50 -4>, angularTarget=<1 0 0>}
      localJointOffset=<12 0 0>
     }]
    planarLimits [{drivingBone {boneName=pelvis}, planeTransform {location=<0 -8.25 -8.25>, rotation=<rot 0 0 -120>}}]
    relativeSpaceBone {boneName=pelvis}
    sphericalLimits [
     {limitRadius=13, drivingBone {boneName=spine_01}, sphereLocalOffset=<0 0 -1>}
     {limitRadius=18.5, drivingBone {boneName=pelvis}, sphereLocalOffset=<-2 5 -1>}
    ]
   }
   n11 = Anim Dynamics - Bone: dyn_jacket_bk_r_01(ComponentPose=@n10, Alpha=1, ExternalForce="0,0,0") {
    angularDampingOverride=1
    angularSpringConstant=600
    bAngularSpring=true
    bGravityOverrideInSimSpace=true
    bOverrideAngularDamping=true
    bUseSphericalLimits=true
    alphaBoolBlend {blendOption=Linear}
    boundBone {boneName=dyn_jacket_bk_r_01}
    componentAppliedLinearAccClamp=<800 800 800>
    componentLinearAccScale=<0.4 0.4 1>
    componentLinearVelScale=<0.8 0.8 2>
    physicsBodyDefinitions [{
      collisionType=CustomSphere
      sphereCollisionRadius=3
      boundBone {boneName=dyn_jacket_bk_r_01}
      boxExtents=<15 20 15>
      constraintSetup {angularLimitsMax=<-2 4 9>, angularLimitsMin=<-2 -40 -4>, angularTarget=<1 0 0>}
      localJointOffset=<-12 0 0>
     }]
    relativeSpaceBone {boneName=pelvis}
    sphericalLimits [
     {limitRadius=10.5, drivingBone {boneName=spine_01}, sphereLocalOffset=<0 1 6>}
     {limitRadius=18.5, drivingBone {boneName=pelvis}, sphereLocalOffset=<0 5 1>}
    ]
   }
   n12 = Anim Dynamics - Bone: dyn_jacket_bk_r_02(ComponentPose=@n11, Alpha=1, ExternalForce="0,0,0") {
    angularDampingOverride=1
    angularSpringConstant=600
    bAngularSpring=true
    bGravityOverrideInSimSpace=true
    bOverrideAngularDamping=true
    bUseSphericalLimits=true
    alphaBoolBlend {blendOption=Linear}
    boundBone {boneName=dyn_jacket_bk_r_02}
    componentAppliedLinearAccClamp=<800 800 800>
    componentLinearAccScale=<0.4 0.4 1>
    componentLinearVelScale=<0.8 0.8 2>
    physicsBodyDefinitions [{
      collisionType=CustomSphere
      sphereCollisionRadius=3
      boundBone {boneName=dyn_jacket_bk_r_02}
      boxExtents=<15 20 15>
      constraintSetup {angularLimitsMax=<-2 50 4>, angularLimitsMin=<-2 -50 -4>, angularTarget=<1 0 0>}
      localJointOffset=<-12 0 0>
     }]
    relativeSpaceBone {boneName=pelvis}
    sphericalLimits [
     {limitRadius=14, drivingBone {boneName=spine_01}, sphereLocalOffset=<0 0 0>}
     {limitRadius=19, drivingBone {boneName=pelvis}, sphereLocalOffset=<0 5 0>}
    ]
   }
   n13 = Break Fort Anim Input Skydiving External Force(FortAnimInput_SkydivingExternalForce=Get SkydivingExternalForceInput()) -> (
    bUseSkydivingVectorForce=~
    SkydivingVectorForce=~
    HeadToPelvisDirection=~
    FloatingMultiplier=~
    FloatingAdditive=~
    DivingMultiplier=~
    DivingAdditive=~
    ParachutingMultiplier=~
    ParachutingAdditive=~
   )
   n14 = RigidBody(ComponentPose=@n12, ExternalForce=@n13.SkydivingVectorForce) {
    bClampLinearTranslationLimitToRefPose=true
    bOverrideWorldGravity=true
    overridePhysicsAsset=&/M_MED_Shiny_Dyn
    simulationSpace=BaseBoneSpace
    alphaBoolBlend {blendOption=Linear}
    baseBoneRef {boneName=root}
    componentAppliedLinearAccClamp=<2000 2000 2000>
    componentLinearAccScale=<0.5 0.3 0.3>
    componentLinearVelScale=<0.5 0.3 0.3>
    externalForce=<-30000 1000 1000>
    overrideWorldGravity {z=-980}
   }
   n15 = "Constraint - dyn_jacket_fr_l_01 (R)"(ComponentPose=@n14, Alpha=1, ConstraintWeights_0=1) {
    alphaBoolBlend {blendOption=Linear}
    boneToModify {boneName=dyn_jacket_fr_l_01}
    constraintSetup [{transformType=Rotation, targetBone {boneName=thigh_l}}]
    constraintWeights [1]
   }
   n16 = "Constraint - dyn_jacket_fr_l_02 (R)"(ComponentPose=@n15, Alpha=1, ConstraintWeights_0=1) {
    alphaBoolBlend {blendOption=Linear}
    boneToModify {boneName=dyn_jacket_fr_l_02}
    constraintSetup [{transformType=Rotation, targetBone {boneName=thigh_l}}]
    constraintWeights [1]
   }
   n17 = Anim Dynamics - Bone: dyn_jacket_fr_l_01(ComponentPose=@n16, Alpha=1, ExternalForce=@n13.SkydivingVectorForce) {
    angularDampingOverride=0.98
    bGravityOverrideInSimSpace=true
    bOverrideAngularDamping=true
    bUseGravityOverride=true
    bUseSphericalLimits=true
    numSolverIterationsPostUpdate=2
    numSolverIterationsPreUpdate=8
    alphaBoolBlend {blendOption=Linear}
    alphaScaleBiasClamp {bClampResult=true, clampMin=0.2}
    boundBone {boneName=dyn_jacket_fr_l_01}
    componentAppliedLinearAccClamp=<1000 1000 1000>
    gravityOverride {z=-980}
    physicsBodyDefinitions [{
      sphereCollisionRadius=2
      boundBone {boneName=dyn_jacket_fr_l_01}
      boxExtents=<20 20 10>
      constraintSetup {angularLimitsMax=<10 1 10>, angularLimitsMin=<-10 -30 -10>, angularTarget=<0 0 0>}
      localJointOffset=<3 0 0>
     }]
    planarLimits [{drivingBone {boneName=thigh_twist_01_l}, planeTransform {location=<0 7 0>, rotation=<rot 0 0 114.9999>}}]
    sphericalLimits [
     {limitRadius=9, drivingBone {boneName=thigh_l}, sphereLocalOffset=<0 0 0>}
     {limitRadius=9, drivingBone {boneName=thigh_twist_01_l}, sphereLocalOffset=<0 0 0>}
     {limitRadius=11, drivingBone {boneName=thigh_twist_01_l}, sphereLocalOffset=<10 0 0>}
    ]
   }
   n18 = Look At - Bone: dyn_jacket_fr_l_01(ComponentPose=@n17, Alpha=Get crouchWalking() -> (crouchWalking:double)) {
    alphaBoolBlend {blendOption=Linear}
    boneToModify {boneName=dyn_jacket_fr_l_01}
    lookAtLocation=<10 10 -10>
    lookAtTarget {boneReference {boneName=calf_l}}
    lookAt_Axis {axis=<1 0>}
    lookUp_Axis {axis {x=0, z=1}}
   }
   n19 = Anim Dynamics - Bone: dyn_jacket_fr_l_02(ComponentPose=@n18, Alpha=1, ExternalForce=@n13.SkydivingVectorForce) {
    angularDampingOverride=0.98
    bGravityOverrideInSimSpace=true
    bOverrideAngularDamping=true
    bUseGravityOverride=true
    bUsePlanarLimit=false
    bUseSphericalLimits=true
    alphaBoolBlend {blendOption=Linear}
    alphaScaleBiasClamp {bClampResult=true, clampMin=0.2}
    boundBone {boneName=dyn_jacket_fr_l_02}
    componentAppliedLinearAccClamp=<1000 1000 1000>
    gravityOverride {z=-980}
    physicsBodyDefinitions [{
      sphereCollisionRadius=3
      boundBone {boneName=dyn_jacket_fr_l_02}
      boxExtents=<20 20 10>
      constraintSetup {angularLimitsMax=<20 1 5>, angularLimitsMin=<-20 -30 -5>, angularTarget=<0 0 0>}
      localJointOffset=<3 0 0>
     }]
    planarLimits [{drivingBone {boneName=thigh_twist_01_l}, planeTransform {location=<0 11 0>, rotation=<rot 5 0 150>}}]
    sphericalLimits [
     {limitRadius=10, drivingBone {boneName=thigh_l}, sphereLocalOffset=<0 0 0>}
     {limitRadius=10, drivingBone {boneName=thigh_twist_01_l}, sphereLocalOffset=<0 0 0>}
     {limitRadius=11, drivingBone {boneName=thigh_twist_01_l}, sphereLocalOffset=<10 0 0>}
     {limitRadius=11, drivingBone {boneName=thigh_l}, sphereLocalOffset=<-6 0 0>}
    ]
   }
   n20 = "Constraint - dyn_jacket_l_01 (R)"!(ComponentPose=@n19, Alpha=0.3) -> (Pose) {
    alphaBoolBlend {blendOption=Linear}
    boneToModify {boneName=dyn_jacket_l_01}
    constraintSetup [{transformType=Rotation, targetBone {boneName=thigh_l}}]
    constraintWeights [1]
   }
   n21 = "Constraint - dyn_jacket_l_02 (R)"!(ComponentPose=@n20, Alpha=0.4) -> (Pose) {
    alphaBoolBlend {blendOption=Linear}
    boneToModify {boneName=dyn_jacket_l_02}
    constraintSetup [{transformType=Rotation, targetBone {boneName=thigh_l}}]
    constraintWeights [1]
   }
   n22 = Get crouchAlpha() -> (crouchAlpha:double)
   n23 = "Transform (Modify) Bone - Bone: dyn_jacket_l_01"!(ComponentPose=@n21, Alpha:float=@n22, Translation="0,0,2", Rotation="0,0,0", Scale="1,1,1") -> (Pose) {
    translationMode=BMM_Additive
    translationSpace=BCS_BoneSpace
    alphaBoolBlend {blendInTime=0.1, blendOption=Linear, blendOutTime=0.1}
    alphaScaleBiasClamp {bInterpResult=true, interpSpeedDecreasing=8, interpSpeedIncreasing=8}
    boneToModify {boneName=dyn_jacket_l_01}
    translation {z=2}
   }
   n24 = "Transform (Modify) Bone - Bone: dyn_jacket_l_02"(ComponentPose=@n23, Alpha=@n22, Translation="0,0,0", Rotation="-30,0,0", Scale="1,1,1") {
    rotationMode=BMM_Additive
    rotationSpace=BCS_BoneSpace
    translationSpace=BCS_BoneSpace
    alphaBoolBlend {blendInTime=0.1, blendOption=Linear, blendOutTime=0.1}
    alphaScaleBiasClamp {bInterpResult=true, interpSpeedDecreasing=8, interpSpeedIncreasing=8}
    boneToModify {boneName=dyn_jacket_l_02}
    rotation {pitch=-30}
   }
   n25 = Anim Dynamics - Bone: dyn_jacket_l_01(ComponentPose=@n24, Alpha=1, ExternalForce=@n13.SkydivingVectorForce) {
    angularDampingOverride=0.98
    bGravityOverrideInSimSpace=true
    bOverrideAngularDamping=true
    bUseGravityOverride=true
    numSolverIterationsPostUpdate=2
    numSolverIterationsPreUpdate=8
    simulationSpace=RootRelative
    alphaBoolBlend {blendOption=Linear}
    boundBone {boneName=dyn_jacket_l_01}
    componentAppliedLinearAccClamp=<1000 1000 1000>
    gravityOverride {z=-980}
    physicsBodyDefinitions [{
      sphereCollisionRadius=0
      boundBone {boneName=dyn_jacket_l_01}
      boxExtents=<20 20 10>
      constraintSetup {angularLimitsMax=<10 -10 7>, angularLimitsMin=<-10 -30 -7>, angularTarget=<0 0 0>}
      localJointOffset=<3 0 0>
     }]
    relativeSpaceBone {boneName=root}
   }
   n26 = Anim Dynamics - Bone: dyn_jacket_l_02(ComponentPose=@n25, Alpha=1, ExternalForce=@n13.SkydivingVectorForce) {
    angularDampingOverride=0.98
    bGravityOverrideInSimSpace=true
    bOverrideAngularDamping=true
    bUseGravityOverride=true
    simulationSpace=RootRelative
    alphaBoolBlend {blendOption=Linear}
    boundBone {boneName=dyn_jacket_l_02}
    componentAppliedLinearAccClamp=<1000 1000 1000>
    gravityOverride {z=-980}
    physicsBodyDefinitions [{
      sphereCollisionRadius=2
      boundBone {boneName=dyn_jacket_l_02}
      boxExtents=<20 20 10>
      constraintSetup {angularLimitsMax=<10 -5 5>, angularLimitsMin=<-10 -30 -5>, angularTarget=<0 0 0>}
      localJointOffset=<3 0 0>
     }]
    planarLimits [{drivingBone {boneName=thigh_twist_01_l}, planeTransform {location=<0 0 -10>, rotation=<rot 0 0 -180>}}]
    relativeSpaceBone {boneName=root}
   }
   n27 = "Transform (Modify) Bone - Bone: dyn_collar_r_MSTR"(ComponentPose=@n26, Alpha=Get arm_side_lift_r() -> (arm_side_lift_r:double), Translation="0,-8,0", Rotation="20,60,-20", Scale="0,0,0") {
    rotationMode=BMM_Additive
    rotationSpace=BCS_BoneSpace
    scaleSpace=BCS_BoneSpace
    translationMode=BMM_Additive
    translationSpace=BCS_BoneSpace
    alphaBoolBlend {blendOption=Linear}
    boneToModify {boneName=dyn_collar_r_MSTR}
    rotation=<rot 20 60 -20>
    scale=<0 0 0>
    translation {y=-8}
   }
   Output Pose(Result=Component To Local(ComponentPose=@n27)) {layerGroup=None}
   Save cached pose 'SavedPose30346'(Pose=Copy Pose From Mesh() {bUseAttachedParent=true}) {cacheName=SavedPose30346}
  ]}
 EventGraph {nodes [
   Event BlueprintUpdateAnimation()
   n1 = GetDeltaTransformFromRefPose(self=GetOwnerSkeletalMeshComponent(), BoneName=upperarm_r, BaseName=clavicle_r)
   n2 = "Absolute (Float)"(A=BreakRotator(InRot=BreakTransform(InTransform=@n1).Rotation).Yaw)
   > Set arm_side_lift_r(arm_side_lift_r:double=float * float(A=MapRangeClamped(Value=@n2, InRangeA=40, InRangeB=120), B=1.5)) -> (Output_Get:double)
   n3 = Get bIsCrouching()
   > Set crouchAlpha(crouchAlpha:double="To Float (Boolean)"(InBool=@n3)) -> (Output_Get:double)
   > Set crouchAlphaInv(crouchAlphaInv:double="To Float (Boolean)"(InBool=NOT Boolean(A=@n3))) -> (Output_Get:double)
   n4 = Get PawnSpeed2D()
   n5 = "float > float"(A=@n4, B=0) {is=title K2Node_CallFunction}
   n6 = Get bIsCrouching()
   > Set crouchWalking(crouchWalking:double="To Float (Boolean)"(InBool=AND Boolean(A=@n5, B=@n6))) -> (Output_Get:double)
   "float >= float"(A=0, B=@n4) {is=title K2Node_CallFunction}
  ]}
]
variables [
 arm_side_lift_r {type=number, default=1}
 crouchAlpha {type=number}
 crouchAlphaInv {type=number}
 crouchWalking {type=number}
]
